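# Zero Point fractures material that builds a glowing, animated crack pattern in inner and outer colors, defaulting to purples.
# A panning turbulence normal distorts world-XY-projected crystal caustics, whose blue channel offsets mesh UVs before a cosine and contrast remap.
# A world-space "Dark Spots" noise mask modulates that pattern, which blends the two colors.
# The emissive-scaled color feeds a ray tracing quality switch between inverse eye adaptation and a separate Lumen multiplier.
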Material /Hera/Sets/ZeroPointFractures/M_TeaCake_ZeroPoint_Fractures
graph {
 properties {bAutomaticallySetUsageInEditor=false, usedShadingModels=Default Lit}
 nodes [
  n1 = ScalarParameter() {defaultValue=1, group=ZeroPoint_Fractures, parameterName=TileMult}
  AppendVector(A=Constant() {r=1}, B=@n1)
  n2 = Multiply(B=Constant2Vector() {r=-0.06}) {constB=0.2}
  n3 = WorldPosition()
  n4 = ScalarParameter() {defaultValue=600, group=ZeroPoint_Fractures, parameterName=Turbulence - Scale}
  n5 = Multiply(
   A=Multiply(A=Divide(A=@n3.XY, B=@n4), B=@n1)
   B=AppendVector(
    A=ScalarParameter() {defaultValue=1, group=ZeroPoint_Fractures, parameterName=Turbulence_Tile_U}
    B=ScalarParameter() {defaultValue=1, group=ZeroPoint_Fractures, parameterName=Turbulence_Tile_V}
   )
  )
  n6 = Time()
  n7 = ScalarParameter() {defaultValue=0.1, group=ZeroPoint_Fractures, parameterName=PanSpeed}
  n8 = TextureSampleParameter2D(UVs=Panner(Coordinate=Add(A=@n2, B=@n5), Time=Multiply(A=@n6, B=@n7)) {speedX=-0.1, speedY=0.05}) {
   group=ZeroPoint_Fractures
   parameterName=Turbulence_Normal
   samplerType=SAMPLERTYPE_Normal
   texture=/Game/Effects/Fort_Effects/Textures/Plasma/T_Turbulence_Seamless_N
  }
  n9 = Subtract(A=Multiply(A=ComponentMask(None=@n8.RGB) {g=true, r=true}) {constB=2})
  n10 = Divide(A=WorldPosition().XY, B=ScalarParameter() {defaultValue=200, group=ZeroPoint_Fractures, parameterName=Scale})
  n11 = TextureSampleParameter2D(UVs=LinearInterpolate(A=@n10, B=Add(A=@n10, B=@n9), Alpha=ScalarParameter() {defaultValue=0.03, group=ZeroPoint_Fractures, parameterName=Distortion})) {
   group=ZeroPoint_Fractures
   parameterName=Turbulence
   samplerType=SAMPLERTYPE_LinearColor
   texture=/Game/Blueprints/Lighting/LightFunctions/CrystalCaustics_M
  }
  n12 = AppendVector(A=@n11.B, B=@n11.B)
  ConstantBiasScale(None=@n12) {scale=2}
  n13 = VectorParameter() {group=ZeroPoint_Fractures, parameterName=Inner Color, defaultValue=<rgba 0.7922 0 1 1>}
  n14 = TextureCoordinate()
  n15 = Subtract(A=Multiply(A=@n12) {constB=2}) {constB=0.5}
  n16 = ComponentMask(None=LinearInterpolate(A=@n14, B=Add(A=@n14, B=@n15), Alpha=ScalarParameter() {defaultValue=0.1, group=ZeroPoint_Fractures, parameterName=Distortion Amount})) {r=true}
  n17 = Saturate(None=Add(A=ScalarParameter() {defaultValue=1.2, group=ZeroPoint_Fractures, parameterName=Offset}, B=Cosine(None=@n16)))
  n18 = NamedRerouteUsage()
  n19 = ScalarParameter() {defaultValue=0.05, group=ZeroPoint_Fractures, parameterName=Dark Spots - Min}
  n20 = Multiply(
   A=MaterialFunctionCall(In=OneMinus(None=@n17), Contrast=ScalarParameter() {defaultValue=1, group=ZeroPoint_Fractures, parameterName=Contrast}) {materialFunction=&CheapContrast}.Result
   B=Saturate(None=Max(A=@n18, B=@n19))
  )
  Multiply(A=@n13.RGB, B=@n20)
  n21 = ScalarParameter() {defaultValue=1000, group=ZeroPoint_Fractures, parameterName=Dark Spots - Scale}
  n22 = Power(
   Base=TextureSample(UVs=Divide(A=@n3.XY, B=@n21)) {texture=/Game/Effects/Fort_Effects/Textures/Plasma/T_Turbulence_Seamless}.R
   Exp=ScalarParameter() {defaultValue=0.5, group=ZeroPoint_Fractures, parameterName=Dark Spots - Power}
  )
  NamedRerouteDeclaration(None=MaterialFunctionCall(In=Saturate(None=@n22), Contrast=ScalarParameter() {defaultValue=0.5, group=ZeroPoint_Fractures, parameterName=Dark Spots - Contrast}) {materialFunction=&CheapContrast}.Result) {name=Dark Spots, nodeColor=<rgba 0.2656 1 0 1>}
  Power()
  n23 = Multiply(
   A=LinearInterpolate(
    A=VectorParameter() {group=ZeroPoint_Fractures, parameterName=Outer Color, defaultValue=<rgba 0.0234 0 0.0573 1>}.RGB
    B=@n13.RGB
    Alpha=@n20
   )
   B=ScalarParameter() {defaultValue=16, group=ZeroPoint_Fractures, parameterName=Emissive Global}
  )
  RayTracingQualitySwitch(
   Normal=EyeAdaptationInverse(LightValueInput=@n23).EyeAdaptationInverse
   RayTraced=Multiply(A=@n23, B=ScalarParameter() {defaultValue=20, group=ZeroPoint_Fractures, parameterName="Emissive Global (Lumen)"})
  )
  n24 = MaterialFunctionCall() {materialFunction=&BreakOutFloat2Components}
  StaticSwitchParameter(
   True=MaterialFunctionCall(X=@n24.R, Y=OneMinus(None=@n24.G)) {materialFunction=&MakeFloat2}.Result
   False=Reroute()
  ) {parameterName=InvertDirection}
  TextureCoordinate()
  TextureCoordinate() {coordinateIndex=1}
  TextureSample() {texture=/Game/Global/Textures/Default/Caustics/T_Turbulence_Seamless}
 ]
}
